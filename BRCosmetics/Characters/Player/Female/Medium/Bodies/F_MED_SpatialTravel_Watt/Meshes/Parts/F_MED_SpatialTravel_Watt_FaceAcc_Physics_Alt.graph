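# Alternate physics asset for the face accessory part of the female medium SpatialTravel Watt cosmetic, covering its dangling dyn_tail bones.
# Capsule bodies on the head, neck, upper spine, clavicles and upper arms form a coarse upper-body volume linked by free-rotation constraints.
# Small capsules on the left, right, back and down tail bones each constrain directly to the head, not in chains, with tight per-bone swing and twist limits.
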PhysicsAsset /BRCosmetics/Characters/Player/Female/Medium/Bodies/F_MED_SpatialTravel_Watt/Meshes/Parts/F_MED_SpatialTravel_Watt_FaceAcc_Physics_Alt
bodies {
 clavicle_l [
  {shapeType=Capsule, length=8, radius=5, shapeName=clavicle_l_capsule, center=<6.4493 0 -0.9773>, rotation=<rot 90 0 0>}
  {
   shapeType=Capsule
   length=8
   radius=3.5
   shapeName=clavicle_l_capsule2
   center=<1.7721 1.9654 3.2081>
   rotation=<rot 8.2396 -24.8875 104.2238>
  }
 ]
 clavicle_r [{shapeType=Capsule, length=8, radius=5, shapeName=clavicle_r_capsule, center=<-5.3886 1.2094 0.5844>, rotation=<rot 90 0 0>}]
 dyn_tail_01_b [{
   shapeType=Capsule
   length=3
   radius=2
   shapeName=dyn_tail_01_b_capsule
   center=<1.7794 -0.3167 -0.5317>
   rotation=<rot 59.3677 -179.9999 -180>
  }]
 dyn_tail_01_l [{
   shapeType=Capsule
   length=5
   radius=1
   shapeName=dyn_tail_01_l_capsule
   center=<2.6537 -0.3639 -1.9913>
   rotation=<rot -7.196 83.4206 -55.0922>
  }]
 dyn_tail_01_r [{
   shapeType=Capsule
   length=3
   radius=1.5
   shapeName=dyn_tail_01_r_capsule
   center=<2.4723 -0.589 -0.7745>
   rotation=<rot 83.6549 179.9994 167.6178>
  }]
 dyn_tail_02_l [{
   shapeType=Capsule
   length=3
   radius=1.5
   shapeName=dyn_tail_02_l_capsule
   center=<2.4657 0.2401 -0.9496>
   rotation=<rot -3.3551 -113.0373 -101.0183>
  }]
 dyn_tail_03_b [{
   shapeType=Capsule
   length=5
   radius=1.5426
   shapeName=dyn_tail_03_b_capsule
   center=<3.1268 -0.3714 0.3996>
   rotation=<rot -71.3641 -41.0307 73.5717>
  }]
 dyn_tail_03_l [
  {
   shapeType=Capsule
   length=4.0964
   radius=1.5
   shapeName=dyn_tail_03_l_capsule
   center=<2.0234 0.506 -0.1399>
   rotation=<rot -87.4132 0.0011 13.9102>
  }
  {
   shapeType=Capsule
   length=4.0964
   radius=1
   shapeName=dyn_tail_03_l_capsule
   center=<1.2866 -2.2548 -0.1587>
   rotation=<rot -86.9774 0.0005 -0.0008>
  }
 ]
 dyn_tail_03_r [{
   shapeType=Capsule
   length=2
   radius=1.9226
   shapeName=dyn_tail_03_r_capsule
   center=<2.3059 -0.4721 0.9114>
   rotation=<rot 77.404 179.9995 -167.8021>
  }]
 dyn_tail_04_r [
  {
   shapeType=Capsule
   length=2
   radius=1.3
   shapeName=dyn_tail_04_r_capsule
   center=<4.7592 0.4145 -0.1634>
   rotation=<rot 2.0383 100.583 -56.2999>
  }
  {shapeType=Capsule, length=2, radius=1, shapeName=dyn_tail_04_r_capsule, center=<-0.1269 -0.0206 -2.7381>, rotation=<rot 0 99.22 6.0863>}
 ]
 dyn_tail_05_b [{
   shapeType=Capsule
   length=3
   radius=1.1923
   shapeName=dyn_tail_05_b_capsule
   center=<1.3908 0.3004 -0.0568>
   rotation=<rot 87.6621 -0.0006 -29.5204>
  }]
 dyn_tail_05_r [
  {
   shapeType=Capsule
   length=4
   radius=1.5293
   shapeName=dyn_tail_05_r_capsule
   center=<5.6129 0.1074 -1.0307>
   rotation=<rot 20.5048 115.9769 -69.6784>
  }
  {
   shapeType=Capsule
   length=4
   radius=1.5
   shapeName=dyn_tail_05_r_capsule
   center=<5.043 -1.7606 -1.2531>
   rotation=<rot 23.5672 127.6329 -90.2239>
  }
 ]
 dyn_tail_06_r [{
   shapeType=Capsule
   length=4
   radius=1.2607
   shapeName=dyn_tail_06_r_capsule
   center=<2.9541 -0.6542 -0.8502>
   rotation=<rot -1.4141 87.8854 -88.8264>
  }]
 dyn_tail_07_r [{
   shapeType=Capsule
   length=3
   radius=1.5
   shapeName=dyn_tail_07_r_capsule
   center=<2.5266 -0.0699 -0.4022>
   rotation=<rot -2.4908 -43.7092 -57.9415>
  }]
 dyn_tail_down_fr [{
   shapeType=Capsule
   length=2
   radius=0.964
   shapeName=dyn_tail_down_fr_capsule
   center=<1.88 -0.3756 0.0317>
   rotation=<rot -87.0868 0.0055 -16.0418>
  }]
 dyn_tail_down_fr1 [{
   shapeType=Capsule
   length=3
   radius=2
   shapeName=dyn_tail_down_fr1_capsule
   center=<1.746 -0.3257 -0.0687>
   rotation=<rot -7.3129 113.0327 -91.789>
  }]
 dyn_tail_down_l [{
   shapeType=Capsule
   length=4
   radius=1.5
   shapeName=dyn_tail_down_l_capsule
   center=<3.3212 0.1892 -1.342>
   rotation=<rot 69.5927 -176.4606 -177.2667>
  }]
 dyn_tail_down_r [{
   shapeType=Capsule
   length=4
   radius=1
   shapeName=dyn_tail_down_r_capsule
   center=<2.3669 0 -0.2019>
   rotation=<rot 70.08 -179.9998 -179.9999>
  }]
 head [{shapeType=Capsule, length=5, radius=7, shapeName=head_capsule, center=<2.1386 2.2283 -0.2911>, rotation=<rot 90 0 0>}]
 neck_01 [{shapeType=Capsule, length=7, radius=5, shapeName=neck_01_capsule, center=<-1.9561 1.0604 0>, rotation=<rot 0 -94.2904 -90>}]
 spine_05 [{shapeType=Capsule, length=5, radius=7, shapeName=spine_05_capsule, center=<5.0605 -1.2679 0>, rotation=<rot 0 -90 0>}]
 upperarm_l [{shapeType=Capsule, length=20, radius=5, shapeName=upperarm_l_capsule, center=<11.1374 -0.4904 -0.1257>, rotation=<rot 90 0 0>}]
 upperarm_r [{
   shapeType=Capsule
   length=20
   radius=4
   shapeName=upperarm_r_capsule
   center=<-11.0311 0.8657 0.1515>
   rotation=<rot 81.3285 -7.5002 -0.0004>
  }]
}
constraints [
 {bone1Name=dyn_tail_down_r, bone2Name=head, swing1LimitDegrees=20, swing2LimitDegrees=20, twistLimitDegrees=10}
 {bone1Name=dyn_tail_01_b, bone2Name=head, swing1LimitDegrees=5, swing2LimitDegrees=30, twistLimitDegrees=5}
 {bone1Name=dyn_tail_down_l, bone2Name=head, swing1LimitDegrees=40, swing2LimitDegrees=40, twistLimitDegrees=10}
 {
  bone1Name=head
  bone2Name=neck_01
  swing1LimitDegrees=45
  swing1Motion=Free
  swing2LimitDegrees=45
  swing2Motion=Free
  twistLimitDegrees=45
  twistMotion=Free
 }
 {
  bone1Name=upperarm_l
  bone2Name=clavicle_l
  swing1LimitDegrees=45
  swing1Motion=Free
  swing2LimitDegrees=45
  swing2Motion=Free
  twistLimitDegrees=45
  twistMotion=Free
 }
 {
  bone1Name=upperarm_r
  bone2Name=clavicle_r
  swing1LimitDegrees=45
  swing1Motion=Free
  swing2LimitDegrees=45
  swing2Motion=Free
  twistLimitDegrees=45
  twistMotion=Free
 }
 {
  bone1Name=clavicle_l
  bone2Name=spine_05
  swing1LimitDegrees=45
  swing1Motion=Free
  swing2LimitDegrees=45
  swing2Motion=Free
  twistLimitDegrees=45
  twistMotion=Free
 }
 {
  bone1Name=clavicle_r
  bone2Name=spine_05
  swing1LimitDegrees=45
  swing1Motion=Free
  swing2LimitDegrees=45
  swing2Motion=Free
  twistLimitDegrees=45
  twistMotion=Free
 }
 {
  bone1Name=neck_01
  bone2Name=spine_05
  swing1LimitDegrees=45
  swing1Motion=Free
  swing2LimitDegrees=45
  swing2Motion=Free
  twistLimitDegrees=45
  twistMotion=Free
 }
 {bone1Name=dyn_tail_down_fr1, bone2Name=head, swing1LimitDegrees=10, swing2LimitDegrees=10, twistLimitDegrees=5}
 {bone1Name=dyn_tail_down_fr, bone2Name=head, swing1LimitDegrees=20, swing2LimitDegrees=20, twistLimitDegrees=5}
 {bone1Name=dyn_tail_02_l, bone2Name=head, swing1LimitDegrees=20, swing2LimitDegrees=20, twistLimitDegrees=10}
 {bone1Name=dyn_tail_06_r, bone2Name=head, swing1LimitDegrees=20, swing2LimitDegrees=20, twistLimitDegrees=10}
 {bone1Name=dyn_tail_05_r, bone2Name=head, swing1LimitDegrees=20, swing2LimitDegrees=20, twistLimitDegrees=5}
 {bone1Name=dyn_tail_04_r, bone2Name=head, swing1LimitDegrees=40, swing2LimitDegrees=40, twistLimitDegrees=5}
 {bone1Name=dyn_tail_03_l, bone2Name=head, swing1LimitDegrees=20, swing2LimitDegrees=20, twistLimitDegrees=5}
 {bone1Name=dyn_tail_01_l, bone2Name=head, swing1LimitDegrees=20, swing2LimitDegrees=20, twistLimitDegrees=10}
 {bone1Name=dyn_tail_03_r, bone2Name=head, swing1LimitDegrees=40, swing2LimitDegrees=40, twistLimitDegrees=5}
 {bone1Name=dyn_tail_03_b, bone2Name=head, swing1LimitDegrees=30, swing2LimitDegrees=10, twistLimitDegrees=10}
 {bone1Name=dyn_tail_01_r, bone2Name=head, swing1LimitDegrees=30, swing2LimitDegrees=15, twistLimitDegrees=5}
 {bone1Name=dyn_tail_07_r, bone2Name=head, swing1LimitDegrees=20, swing2LimitDegrees=30, twistLimitDegrees=5}
 {bone1Name=dyn_tail_05_b, bone2Name=head, swing1LimitDegrees=20, swing2LimitDegrees=20, twistLimitDegrees=5}
]
